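# Physics asset for the M_MED_Kitty_Warrior_Light backpack cosmetic, giving its pack and two dangling straps simulation bodies.
# A long capsule covers the pack bone, and a small capsule sits on each of the top and bottom strap bones.
# Each strap is constrained directly to the pack with a moderate first swing limit, a narrow second swing and almost no twist, so it moves mostly along one axis.
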PhysicsAsset /BRCosmetics/Accessories/FORT_Backpacks/Backpack_M_MED_Kitty_Warrior_Light/Meshes/M_MED_Kitty_Warrior_Light_Pack_Physics
bodies {
 dyn_strap_bot [{shapeType=Capsule, length=1.5588, radius=0.8, shapeName=dyn_strap_bot_capsule, center=<1 0 0>, rotation=<rot 90 0 0>}]
 dyn_strap_top [{shapeType=Capsule, length=1.5588, radius=0.8, shapeName=dyn_strap_top_capsule, center=<1 0 0>, rotation=<rot 90 0 0>}]
 pack [{
   shapeType=Capsule
   length=35.8673
   radius=2.8373
   shapeName=pack_capsule
   center=<1.9964 -0.7618 -4.0104>
   rotation=<rot 0.7111 -89.9999 20.6986>
  }]
}
constraints [
 {bone1Name=dyn_strap_top, bone2Name=pack, swing1LimitDegrees=30, swing2LimitDegrees=5, twistLimitDegrees=1}
 {bone1Name=dyn_strap_bot, bone2Name=pack, swing1LimitDegrees=30, swing2LimitDegrees=5, twistLimitDegrees=1}
]
